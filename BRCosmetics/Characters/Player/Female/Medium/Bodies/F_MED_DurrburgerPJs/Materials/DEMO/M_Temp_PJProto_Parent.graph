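# Prototype parent material for the Durrburger PJs character cosmetic, recoloring base color through three mask-channel tints, with an optional emoji sticker pattern behind a static switch.
# The sticker tiles across UVs with an offset on alternating rows, and a seeded per-cell hash varies its rotation and scale.
# The work sits between the shared CharacterParent_2023 PreFX and PostFX functions, overriding only base color before PostFX.
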Material /BRCosmetics/Characters/Player/Female/Medium/Bodies/F_MED_DurrburgerPJs/Materials/DEMO/M_Temp_PJProto_Parent
graph {
 expressions {
  MaterialExpressionAdd_0 {
   class=MaterialExpressionAdd
   inputs [
    A {from=MaterialExpressionDivide_0}
    B {from=""}
   ]
   properties {constB=0.01}
  }
  MaterialExpressionAdd_1 {
   class=MaterialExpressionAdd
   inputs [
    A {from=MaterialExpressionMultiply_2}
    B {from=MaterialExpressionConvert_0}
   ]
  }
  MaterialExpressionAdd_2 {
   class=MaterialExpressionAdd
   inputs [
    A {from=MaterialExpressionAdd_1}
    B {from=MaterialExpressionAppendVector_0}
   ]
  }
  MaterialExpressionAdd_3 {
   class=MaterialExpressionAdd
   inputs [
    A {from=MaterialExpressionMultiply_3}
    B {from=MaterialExpressionScalarParameter_6}
   ]
  }
  MaterialExpressionAdd_4 {
   class=MaterialExpressionAdd
   inputs [
    A {from=MaterialExpressionDivide_3}
    B {from=MaterialExpressionMultiply_7}
   ]
  }
  MaterialExpressionAdd_5 {
   class=MaterialExpressionAdd
   inputs [
    A {from=MaterialExpressionMultiply_2}
    B {from=""}
   ]
   properties {constB=0.5}
  }
  MaterialExpressionAppendVector_0 {
   class=MaterialExpressionAppendVector
   inputs [
    A {from=MaterialExpressionMultiply_5, output=""}
    B {from=MaterialExpressionConstant_2, output=""}
   ]
  }
  MaterialExpressionChannelMaskParameter_0 {
   class=MaterialExpressionChannelMaskParameter
   inputs [None {from=MaterialExpressionNamedRerouteUsage_1, output=""}]
   properties {parameterName=PrimaryColorMaskChannel}
  }
  MaterialExpressionChannelMaskParameter_1 {
   class=MaterialExpressionChannelMaskParameter
   inputs [None {from=MaterialExpressionNamedRerouteUsage_2, output=""}]
   properties {maskChannel=Green, parameterName=SecondaryColorMaskChannel, defaultValue=<rgba 0 1 0 0>}
  }
  MaterialExpressionChannelMaskParameter_2 {
   class=MaterialExpressionChannelMaskParameter
   inputs [None {from=MaterialExpressionNamedRerouteUsage_3, output=""}]
   properties {maskChannel=Blue, parameterName=TertiaryColorMaskChannel, defaultValue=<rgba 0 0 1 0>}
  }
  MaterialExpressionComponentMask_0 {class=MaterialExpressionComponentMask, inputs [None {from=MaterialExpressionNamedRerouteUsage_1}], properties {r=true}}
  MaterialExpressionComponentMask_1 {class=MaterialExpressionComponentMask, inputs [None {from=MaterialExpressionNamedRerouteUsage_2}], properties {g=true}}
  MaterialExpressionComponentMask_2 {class=MaterialExpressionComponentMask, inputs [None {from=MaterialExpressionNamedRerouteUsage_3}], properties {b=true}}
  MaterialExpressionComponentMask_3 {class=MaterialExpressionComponentMask, inputs [None {from=MaterialExpressionConstantBiasScale_0}], properties {r=true}}
  MaterialExpressionComponentMask_4 {class=MaterialExpressionComponentMask, inputs [None {from=MaterialExpressionNamedRerouteUsage_4}], properties {g=true}}
  MaterialExpressionComponentMask_5 {class=MaterialExpressionComponentMask, inputs [None {from=MaterialExpressionStep_0}], properties {g=true}}
  MaterialExpressionComponentMask_6 {class=MaterialExpressionComponentMask, inputs [None {from=MaterialExpressionConstantBiasScale_0}], properties {g=true}}
  MaterialExpressionConstantBiasScale_0 {class=MaterialExpressionConstantBiasScale, inputs [None {from=MaterialExpressionMaterialFunctionCall_4, output=Vec2}]}
  MaterialExpressionConstant_0 {class=MaterialExpressionConstant, properties {r=2}}
  MaterialExpressionConstant_2 {class=MaterialExpressionConstant}
  MaterialExpressionConstant_3 {class=MaterialExpressionConstant}
  MaterialExpressionConstant_4 {class=MaterialExpressionConstant, properties {r=1}}
  MaterialExpressionConvert_0 {
   class=MaterialExpressionConvert
   inputs [
    None {from=MaterialExpressionScalarParameter_4}
    None {from=MaterialExpressionScalarParameter_5}
   ]
   properties {
    nodeName=Make Float2
    convertInputs [
     {
      type=Scalar
      expressionInput {
       expression=&self:MaterialExpressionScalarParameter_4
       inputName=None
       mask=0
       maskA=0
       maskB=0
       maskG=0
       maskR=0
       outputIndex=0
      }
     }
     {
      type=Scalar
      expressionInput {
       expression=&self:MaterialExpressionScalarParameter_5
       inputName=None
       mask=0
       maskA=0
       maskB=0
       maskG=0
       maskR=0
       outputIndex=0
      }
     }
    ]
    convertOutputs [{type=Vector2}]
   }
  }
  MaterialExpressionDivide_0 {
   class=MaterialExpressionDivide
   inputs [
    A {from=MaterialExpressionFloor_0}
    B {from=MaterialExpressionScalarParameter_0}
   ]
  }
  MaterialExpressionDivide_2 {
   class=MaterialExpressionDivide
   inputs [
    A {from=MaterialExpressionAdd_3}
    B {from=""}
   ]
   properties {constB=360}
  }
  MaterialExpressionDivide_3 {
   class=MaterialExpressionDivide
   inputs [
    A {from=MaterialExpressionScalarParameter_1}
    B {from=""}
   ]
   properties {constB=100}
  }
  MaterialExpressionDivide_4 {
   class=MaterialExpressionDivide
   inputs [
    A {from=MaterialExpressionScalarParameter_9}
    B {from=""}
   ]
   properties {constB=100}
  }
  MaterialExpressionFloor_0 {class=MaterialExpressionFloor, inputs [None {from=MaterialExpressionAdd_2, output=""}], properties {}}
  MaterialExpressionFmod_0 {
   class=MaterialExpressionFmod
   inputs [
    A {from=MaterialExpressionAdd_1, output=""}
    B {from=MaterialExpressionConstant_0, output=""}
   ]
   properties {}
  }
  MaterialExpressionFrac_1 {class=MaterialExpressionFrac, inputs [None {from=MaterialExpressionAdd_2, output=""}], properties {}}
  MaterialExpressionGetMaterialAttributes_0 {
   class=MaterialExpressionGetMaterialAttributes
   inputs [None {from=MaterialExpressionMaterialFunctionCall_0, output=Result}]
   properties {attributeGetTypes [#1, #2, #3]}
  }
  MaterialExpressionLinearInterpolate_0 {
   class=MaterialExpressionLinearInterpolate
   inputs [
    A {from=""}
    B {from=MaterialExpressionVectorParameter_0, output=RGB}
    Alpha {from=MaterialExpressionChannelMaskParameter_0, output=R}
   ]
   properties {constA=1}
  }
  MaterialExpressionLinearInterpolate_1 {
   class=MaterialExpressionLinearInterpolate
   inputs [
    A {from=MaterialExpressionLinearInterpolate_0}
    B {from=MaterialExpressionVectorParameter_1, output=RGB}
    Alpha {from=MaterialExpressionChannelMaskParameter_1, output=R}
   ]
   properties {constA=1}
  }
  MaterialExpressionLinearInterpolate_2 {
   class=MaterialExpressionLinearInterpolate
   inputs [
    A {from=MaterialExpressionLinearInterpolate_1}
    B {from=MaterialExpressionVectorParameter_2, output=RGB}
    Alpha {from=MaterialExpressionChannelMaskParameter_2, output=R}
   ]
   properties {constA=1}
  }
  MaterialExpressionLinearInterpolate_3 {
   class=MaterialExpressionLinearInterpolate
   inputs [
    A {from=MaterialExpressionLinearInterpolate_2}
    B {from=MaterialExpressionNamedRerouteUsage_5}
    Alpha {from=MaterialExpressionNamedRerouteUsage_6}
   ]
  }
  MaterialExpressionMaterialFunctionCall_0 {
   class=MaterialExpressionMaterialFunctionCall
   inputs [
    Overide Texture Set UVs {from=""}
    Override Tech Art Mask UVs {from=""}
    Override Basic Emissive UVs {from=""}
   ]
   properties {materialFunction=&MF_CharacterParent_2023_PreFX}
  }
  MaterialExpressionMaterialFunctionCall_1 {
   class=MaterialExpressionMaterialFunctionCall
   inputs [
    Material {from=MaterialExpressionSetMaterialAttributes_1}
    M {from=MaterialExpressionMaterialFunctionCall_0, output=M}
    Pre Skinned Local Position {from=MaterialExpressionMaterialFunctionCall_0, output=M}
    Distance Blend {from=MaterialExpressionMaterialFunctionCall_0, output=M}
    Fresnel {from=MaterialExpressionMaterialFunctionCall_0, output=M}
   ]
   properties {materialFunction=&MF_CharacterParent_2023_PostFX}
  }
  MaterialExpressionMaterialFunctionCall_2 {
   class=MaterialExpressionMaterialFunctionCall
   inputs [
    UVs {from=MaterialExpressionFrac_1}
    Texture Scale {from=MaterialExpressionAdd_4}
   ]
   properties {materialFunction=&ScaleUVsByCenter}
  }
  MaterialExpressionMaterialFunctionCall_4 {
   class=MaterialExpressionMaterialFunctionCall
   inputs [
    Seed {from=MaterialExpressionScalarParameter_2}
    Float {from=""}
    Vec3 {from=""}
    Vec2 {from=MaterialExpressionAdd_0}
   ]
   properties {materialFunction=&MF_Hash}
  }
  MaterialExpressionMaterialFunctionCall_6 {
   class=MaterialExpressionMaterialFunctionCall
   inputs [
    Input {from=MaterialExpressionComponentMask_6}
    Input Low {from=MaterialExpressionConstant_3}
    Input High {from=MaterialExpressionConstant_4}
    Target Low {from=MaterialExpressionSubtract_1}
    Target High {from=MaterialExpressionSubtract_0}
   ]
   properties {materialFunction=&RemapValueRange}
  }
  MaterialExpressionMultiply_0 {
   class=MaterialExpressionMultiply
   inputs [
    A {from=MaterialExpressionNamedRerouteUsage_0}
    B {from=MaterialExpressionStaticSwitchParameter_0}
   ]
  }
  MaterialExpressionMultiply_1 {
   class=MaterialExpressionMultiply
   inputs [
    A {from=MaterialExpressionDivide_2}
    B {from=""}
   ]
   properties {constB=25.1}
  }
  MaterialExpressionMultiply_2 {
   class=MaterialExpressionMultiply
   inputs [
    A {from=MaterialExpressionTextureCoordinate_1}
    B {from=MaterialExpressionScalarParameter_0}
   ]
  }
  MaterialExpressionMultiply_3 {
   class=MaterialExpressionMultiply
   inputs [
    A {from=MaterialExpressionComponentMask_3}
    B {from=MaterialExpressionScalarParameter_3}
   ]
  }
  MaterialExpressionMultiply_4 {
   class=MaterialExpressionMultiply
   inputs [
    A {from=MaterialExpressionTextureSampleParameter2D_0, output=A}
    B {from=MaterialExpressionComponentMask_4}
   ]
  }
  MaterialExpressionMultiply_5 {
   class=MaterialExpressionMultiply
   inputs [
    A {from=MaterialExpressionScalarParameter_7}
    B {from=MaterialExpressionNamedRerouteUsage_8}
   ]
  }
  MaterialExpressionMultiply_7 {
   class=MaterialExpressionMultiply
   inputs [
    A {from=MaterialExpressionMaterialFunctionCall_6, output=Result}
    B {from=MaterialExpressionDivide_4}
   ]
  }
  MaterialExpressionNamedRerouteDeclaration_0 {
   class=MaterialExpressionNamedRerouteDeclaration
   inputs [None {from=MaterialExpressionGetMaterialAttributes_0, output=Base Color}]
   properties {name=BaseColor, nodeColor=<rgba 0.875 1 0 1>}
  }
  MaterialExpressionNamedRerouteDeclaration_1 {
   class=MaterialExpressionNamedRerouteDeclaration
   inputs [None {from=MaterialExpressionMaterialFunctionCall_0, output=Tech Art Mask}]
   properties {name=TechArtMask, nodeColor=<rgba 0.8047 0 1 1>}
  }
  MaterialExpressionNamedRerouteDeclaration_2 {
   class=MaterialExpressionNamedRerouteDeclaration
   inputs [None {from=MaterialExpressionTextureSampleParameter2D_0, output=RGB}]
   properties {name=Emoji_RGB, nodeColor=<rgba 0 0.1562 1 1>}
  }
  MaterialExpressionNamedRerouteDeclaration_3 {
   class=MaterialExpressionNamedRerouteDeclaration
   inputs [None {from=MaterialExpressionMultiply_4}]
   properties {name=Emoji_Alpha, nodeColor=<rgba 0 0.1562 1 1>}
  }
  MaterialExpressionNamedRerouteDeclaration_4 {
   class=MaterialExpressionNamedRerouteDeclaration
   inputs [None {from=MaterialExpressionComponentMask_5}]
   properties {name=Interstice, nodeColor=<rgba 1 0.7969 0 1>}
  }
  MaterialExpressionNamedRerouteDeclaration_5 {
   class=MaterialExpressionNamedRerouteDeclaration
   inputs [None {from=MaterialExpressionMultiply_0}]
   properties {name=BaseColor_OUT, nodeColor=<rgba 1 0.0703 0 1>}
  }
  MaterialExpressionNamedRerouteDeclaration_6 {
   class=MaterialExpressionNamedRerouteDeclaration
   inputs [None {from=MaterialExpressionMultiply_3}]
   properties {name=RandomRotation, nodeColor=<rgba 0 1 0.9531 1>}
  }
  MaterialExpressionNamedRerouteUsage_0 {class=MaterialExpressionNamedRerouteUsage}
  MaterialExpressionNamedRerouteUsage_1 {class=MaterialExpressionNamedRerouteUsage}
  MaterialExpressionNamedRerouteUsage_2 {class=MaterialExpressionNamedRerouteUsage}
  MaterialExpressionNamedRerouteUsage_3 {class=MaterialExpressionNamedRerouteUsage}
  MaterialExpressionNamedRerouteUsage_4 {class=MaterialExpressionNamedRerouteUsage}
  MaterialExpressionNamedRerouteUsage_5 {class=MaterialExpressionNamedRerouteUsage}
  MaterialExpressionNamedRerouteUsage_6 {class=MaterialExpressionNamedRerouteUsage}
  MaterialExpressionNamedRerouteUsage_8 {class=MaterialExpressionNamedRerouteUsage}
  MaterialExpressionNamedRerouteUsage_9 {class=MaterialExpressionNamedRerouteUsage}
  MaterialExpressionRotator_0 {
   class=MaterialExpressionRotator
   inputs [
    Coordinate {from=MaterialExpressionMaterialFunctionCall_2, output=UVs}
    Time {from=MaterialExpressionMultiply_1, output=""}
   ]
   properties {}
  }
  MaterialExpressionSaturate_0 {class=MaterialExpressionSaturate, inputs [None {from=MaterialExpressionComponentMask_6}]}
  MaterialExpressionScalarParameter_0 {class=MaterialExpressionScalarParameter, properties {defaultValue=10, parameterName=EmojiTile}}
  MaterialExpressionScalarParameter_1 {class=MaterialExpressionScalarParameter, properties {defaultValue=40, parameterName=EmojiScale}}
  MaterialExpressionScalarParameter_2 {class=MaterialExpressionScalarParameter, properties {defaultValue=3, parameterName=RandomSeed}}
  MaterialExpressionScalarParameter_3 {class=MaterialExpressionScalarParameter, properties {defaultValue=1, parameterName=EmojiRandomRotation}}
  MaterialExpressionScalarParameter_4 {class=MaterialExpressionScalarParameter, properties {parameterName=OffsetX}}
  MaterialExpressionScalarParameter_5 {class=MaterialExpressionScalarParameter, properties {parameterName=OffsetY}}
  MaterialExpressionScalarParameter_6 {class=MaterialExpressionScalarParameter, properties {parameterName=EmojiRotation}}
  MaterialExpressionScalarParameter_7 {class=MaterialExpressionScalarParameter, properties {defaultValue=0.5, parameterName=IntersticeX}}
  MaterialExpressionScalarParameter_9 {class=MaterialExpressionScalarParameter, properties {parameterName=EmojiRandomScale}}
  MaterialExpressionSetMaterialAttributes_0 {
   class=MaterialExpressionSetMaterialAttributes
   inputs [
    MaterialAttributes {from=MaterialExpressionGetMaterialAttributes_0, output=MaterialAttributes}
    Roughness {from=MaterialExpressionGetMaterialAttributes_0, output=MaterialAttributes}
    Metallic {from=MaterialExpressionGetMaterialAttributes_0, output=MaterialAttributes}
   ]
   properties {attributeSetTypes [#1, #2]}
  }
  MaterialExpressionSetMaterialAttributes_1 {
   class=MaterialExpressionSetMaterialAttributes
   inputs [
    MaterialAttributes {from=MaterialExpressionSetMaterialAttributes_0, output=""}
    Base Color {from=MaterialExpressionNamedRerouteUsage_9, output=""}
   ]
   properties {attributeSetTypes [#3]}
  }
  MaterialExpressionStaticSwitchParameter_0 {
   class=MaterialExpressionStaticSwitchParameter
   inputs [
    True {from=MaterialExpressionLinearInterpolate_3, output=""}
    False {from=MaterialExpressionLinearInterpolate_2, output=""}
   ]
   properties {defaultValue=true, parameterName=UseEmoji}
  }
  MaterialExpressionStep_0 {
   class=MaterialExpressionStep
   inputs [
    Y {from=MaterialExpressionFmod_0, output=""}
    X {from="", output=""}
   ]
   properties {}
  }
  MaterialExpressionSubtract_0 {
   class=MaterialExpressionSubtract
   inputs [
    A {from=""}
    B {from=MaterialExpressionDivide_3}
   ]
  }
  MaterialExpressionSubtract_1 {
   class=MaterialExpressionSubtract
   inputs [
    A {from=""}
    B {from=MaterialExpressionDivide_3}
   ]
   properties {constA=0.2, constB=0.2}
  }
  MaterialExpressionSubtract_2 {
   class=MaterialExpressionSubtract
   inputs [
    A {from=MaterialExpressionTextureCoordinate_1}
    B {from=""}
   ]
   properties {constB=0.5}
  }
  MaterialExpressionTextureCoordinate_1 {class=MaterialExpressionTextureCoordinate}
  MaterialExpressionTextureSampleParameter2D_0 {
   class=MaterialExpressionTextureSampleParameter2D
   inputs [
    UVs {from=MaterialExpressionRotator_0, output=""}
    Apply View MipBias {from="", output=""}
   ]
   properties {
    parameterName=Sticker
    samplerSource=SSM_Clamp_WorldGroupSettings
    texture=/BRCosmetics/UI/Foundation/Textures/Icons/Emoji/Season_3/T-Item-Emoji-200IQPlay
   }
  }
  MaterialExpressionVectorParameter_0 {class=MaterialExpressionVectorParameter, properties {parameterName=PrimaryColor, defaultValue=<rgba 1 0 0 1>}}
  MaterialExpressionVectorParameter_1 {class=MaterialExpressionVectorParameter, properties {parameterName=SecondaryColor, defaultValue=<rgba 0 1 0 1>}}
  MaterialExpressionVectorParameter_2 {class=MaterialExpressionVectorParameter, properties {parameterName=TertiaryColor, defaultValue=<rgba 0 0 1 1>}}
 }
 properties {
  bUseMaterialAttributes=true
  bUsedWithClothing=true
  bUsedWithMeshDeformer=true
  bUsedWithMorphTargets=true
  bUsedWithNanite=true
  bUsedWithNiagaraMeshParticles=true
  bUsedWithSkeletalMesh=true
  pixelDepthOffsetMode=PDOM_Legacy
  shadingModel=MSM_FromMaterialExpression
  usedShadingModels=Default Lit | Subsurface | Clear Coat | Subsurface Profile | Cloth
 }
}
